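# Editor utility menu entry that cheat-completes the quests selected as assets in the Unreal editor.
# On Execute it loops over the selected asset data and keeps only assets whose class derives from FortQuestItemDefinition.
# For each match it builds a "cheat CompleteAllQuestsByName" console command from the asset name, runs it, and prints the command.
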EditorUtilityBlueprint /Game/NeverCook/EditorUtilities/QuestTools/QuestUtilities/CheatMenus/QuestCheatMenuEntries/CompleteQuest
graphs [EventGraph {nodes [
   Event Execute!() -> (OutputDelegate, then, Context:"Tool Menu Context Structure (by ref)") {is=title K2Node_Event}
   > n1 = GetSelectedAssetData()
   > n2 = For Each Loop(Array:Array of Asset Data Structures=@n1) -> (Array Element:Asset Data Structure)
   > Branch(Condition=ClassIsChildOf(
     TestClass=GetClass!(InAssetData:"Asset Data Structure (by ref)"=@n2) -> (ReturnValue:Object Class Reference)
     ParentClass=/Script/FortniteGame.FortQuestItemDefinition
    ))
   n3 = Append(A="cheat CompleteAllQuestsByName  ", B="To String (Name)"(InName=Break Asset Data(AssetData=@n2).AssetName))
   > ExecuteConsoleCommand(Command=@n3)
   > PrintString(InString=Append(A="Executing Command: ", B=@n3))
  ]}]
variables []
